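# Post-process style AnimBlueprint for the F_MED_Cadet female body, adding secondary motion to its accessories.
# It copies the parent mesh pose, runs a deformer Control Rig, shifts belt bones by crouch and right-thigh alphas, and constrains accessories to weighted bones.
# The cached result feeds AnimDynamics for belt, pouches, necklace and shoelaces, with a brief necklace adjustment on weapon change.
# EventGraph computes those alphas, plus NeckTurn and GravityOverride, which the AnimGraph leaves unused.
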AnimBlueprint /BRCosmetics/Characters/Player/Female/Medium/Bodies/F_MED_Cadet/Meshes/F_MED_Cadet_AnimBP
target_skeleton=&/F_MED_Cadet_Skeleton
graphs [
 AnimGraph {nodes [
   n1 = Anim Dynamics - Bone: dyn_belt_rFR_01(ComponentPose=Local To Component(LocalPose=Use cached pose 'PreSim'()), Alpha=0.7, ExternalForce="0,0,0") {
    angularDampingOverride=0.96
    bUseGravityOverride=true
    linearDampingOverride=0.995
    alphaBoolBlend {blendOption=Linear}
    boundBone {boneName=dyn_belt_rFR_01}
    componentAppliedLinearAccClamp=<3000 3000 3000>
    componentLinearAccScale=<0.1 0.1 0.3>
    componentLinearVelScale=<0.1 0.1 0.3>
    gravityOverride {z=-940}
    physicsBodyDefinitions [{
      sphereCollisionRadius=0
      boundBone {boneName=dyn_belt_rFR_01}
      boxExtents=<20 20 3>
      constraintSetup {angularLimitsMax=<0 0 10>, angularLimitsMin=<0 0 0>, angularTarget=<0 0 0>, linearAxesMax=<0 2 0>}
      localJointOffset=<3 0 0>
     }]
    relativeSpaceBone {boneName=pelvis}
   }
   n2 = Anim Dynamics - Bone: dyn_belt_rBK_01(ComponentPose=@n1, Alpha=0.6, ExternalForce="0,0,0") {
    angularDampingOverride=0.96
    bOverrideLinearDamping=true
    bUseGravityOverride=true
    linearDampingOverride=0.995
    alphaBoolBlend {blendOption=Linear}
    boundBone {boneName=dyn_belt_rBK_01}
    componentAppliedLinearAccClamp=<3000 3000 3000>
    componentLinearAccScale=<0.1 0.1 0.3>
    componentLinearVelScale=<0.1 0.1 0.3>
    gravityOverride {z=-940}
    physicsBodyDefinitions [{
      sphereCollisionRadius=0
      boundBone {boneName=dyn_belt_rBK_01}
      boxExtents=<20 20 3>
      constraintSetup {angularLimitsMax=<0 0 10>, angularLimitsMin=<0 0 0>, angularTarget=<0 0 0>, linearAxesMax=<0 1 0>}
      localJointOffset=<3 0 0>
     }]
    relativeSpaceBone {boneName=pelvis}
   }
   n3 = "Constraint - dyn_beltPouch_a_MSTR (P)"(ComponentPose=@n2, Alpha=0.2, ConstraintWeights_0=1) {
    alphaBoolBlend {blendOption=Linear}
    boneToModify {boneName=dyn_beltPouch_a_MSTR}
    constraintSetup [{targetBone {boneName=dyn_belt_rBK_01}}]
    constraintWeights [1]
   }
   n4 = "Constraint - dyn_beltPouch_b_MSTR (P)"(ComponentPose=@n3, Alpha=0.3, ConstraintWeights_0=1) {
    alphaBoolBlend {blendOption=Linear}
    boneToModify {boneName=dyn_beltPouch_b_MSTR}
    constraintSetup [{targetBone {boneName=dyn_belt_rBK_01}}]
    constraintWeights [1]
   }
   n5 = Anim Dynamics - Bone: dyn_beltPouch_a(ComponentPose=@n4, Alpha=1, ExternalForce="0,0,0") {
    angularDampingOverride=0.995
    bOverrideAngularDamping=true
    bUseGravityOverride=true
    alphaBoolBlend {blendOption=Linear}
    boundBone {boneName=dyn_beltPouch_a}
    componentAppliedLinearAccClamp=<3000 3000 3000>
    componentLinearAccScale=<0.3 0.3 0.35>
    componentLinearVelScale=<0.3 0.3 0.35>
    gravityOverride {z=-980}
    physicsBodyDefinitions [{
      sphereCollisionRadius=0
      boundBone {boneName=dyn_beltPouch_a}
      boxExtents=<18 5.5 22>
      constraintSetup {angularLimitsMax=<0 5 18>, angularLimitsMin=<0 -5 -2>, angularTarget=<0 0 0>}
      localJointOffset=<1.5 0 0>
     }]
    relativeSpaceBone {boneName=pelvis}
   }
   n6 = Anim Dynamics - Bone: dyn_beltPouch_b(ComponentPose=@n5, Alpha=1, ExternalForce="0,0,0") {
    angularDampingOverride=0.999
    bOverrideAngularDamping=true
    bUseGravityOverride=true
    alphaBoolBlend {blendOption=Linear}
    boundBone {boneName=dyn_beltPouch_b}
    componentAppliedLinearAccClamp=<3000 3000 3000>
    componentLinearAccScale=<0.3 0.3 0.35>
    componentLinearVelScale=<0.3 0.3 0.35>
    gravityOverride {z=-980}
    physicsBodyDefinitions [{
      sphereCollisionRadius=0
      boundBone {boneName=dyn_beltPouch_b}
      boxExtents=<20 6 25>
      constraintSetup {angularLimitsMax=<0 5 21>, angularLimitsMin=<0 -5 -2>, angularTarget=<0 0 0>}
      localJointOffset=<1.75 0 0>
     }]
    relativeSpaceBone {boneName=pelvis}
   }
   n7 = "Transform (Modify) Bone - Bone: dyn_necklace_NECKAdj"!(
    ComponentPose=@n6
    bAlphaBoolEnabled:bool=Get isWeaponSwapping() -> (isWeaponSwapping:bool)
    Translation="0,-0.7,0"
    Rotation="0,0,0"
    Scale="1.3,1.4,1.3"
   ) -> (Pose) {
    alphaInputType=Bool
    rotationMode=BMM_Additive
    rotationSpace=BCS_BoneSpace
    scaleMode=BMM_Additive
    scaleSpace=BCS_BoneSpace
    translationMode=BMM_Additive
    translationSpace=BCS_BoneSpace
    alphaBoolBlend {blendInTime=0.1, blendOption=Linear, blendOutTime=0.1}
    alphaScaleBiasClamp {inRange {min=0.15}}
    boneToModify {boneName=dyn_necklace_NECKAdj}
    scale=<1.3 1.4 1.3>
    translation {y=-0.7}
   }
   n8 = "Transform (Modify) Bone - Bone: dyn_necklace_NECKAdj"!(ComponentPose=@n7, Alpha:float=1, Translation="0,0,2", Rotation="0,0,0", Scale="1.175,1.2,1.15") -> (Pose) {
    rotationMode=BMM_Additive
    rotationSpace=BCS_BoneSpace
    scaleMode=BMM_Additive
    scaleSpace=BCS_BoneSpace
    translationMode=BMM_Additive
    translationSpace=BCS_BoneSpace
    alphaBoolBlend {blendInTime=0.1, blendOption=Linear, blendOutTime=0.1}
    alphaScaleBiasClamp {inRange {min=0.15}}
    boneToModify {boneName=dyn_necklace_NECKAdj}
    scale=<1.175 1.2 1.15>
    translation {z=2}
   }
   n9 = Anim Dynamics - Bone: dyn_necklace_01(ComponentPose=@n8, Alpha=0.7, ExternalForce="0,0,0") {
    angularDampingOverride=0.9999
    bOverrideAngularDamping=true
    bUseGravityOverride=true
    bUsePlanarLimit=false
    alphaBoolBlend {blendOption=Linear}
    boundBone {boneName=dyn_necklace_01}
    componentAppliedLinearAccClamp=<1000 1000 1000>
    componentLinearAccScale=<0.1 0.1 0.2>
    componentLinearVelScale=<0.1 0.1 0.2>
    gravityOverride {z=-980}
    physicsBodyDefinitions [{
      sphereCollisionRadius=0
      boundBone {boneName=dyn_necklace_01}
      boxExtents=<12 4 12>
      constraintSetup {angularLimitsMax=<0 4 5>, angularLimitsMin=<0 -4 2>, angularTarget=<1 0 0>}
      localJointOffset=<2 0 0>
     }]
   }
   n10 = Anim Dynamics - Bone: dyn_necklace_02(ComponentPose=@n9, Alpha=0.7, ExternalForce="0,0,0") {
    angularDampingOverride=0.9999
    bOverrideAngularDamping=true
    bUseGravityOverride=true
    bUsePlanarLimit=false
    simulationSpace=RootRelative
    alphaBoolBlend {blendOption=Linear}
    boundBone {boneName=dyn_necklace_02}
    componentAppliedLinearAccClamp=<1000 1000 1000>
    componentLinearAccScale=<0.1 0.1 0.2>
    componentLinearVelScale=<0.1 0.1 0.2>
    gravityOverride {z=-980}
    physicsBodyDefinitions [{
      sphereCollisionRadius=0
      boundBone {boneName=dyn_necklace_02}
      boxExtents=<14 4 14>
      constraintSetup {angularLimitsMax=<0 3 10>, angularLimitsMin=<0 -3 0>, angularTarget=<1 0 0>}
      localJointOffset=<2 0 0>
     }]
   }
   n11 = Anim Dynamics - Bone: dyn_necklace_03(ComponentPose=@n10, Alpha=0.7, ExternalForce="0,0,0") {
    angularDampingOverride=0.9999
    bOverrideAngularDamping=true
    bUseGravityOverride=true
    bUsePlanarLimit=false
    simulationSpace=RootRelative
    alphaBoolBlend {blendOption=Linear}
    boundBone {boneName=dyn_necklace_03}
    componentAppliedLinearAccClamp=<1000 1000 1000>
    componentLinearAccScale=<0.1 0.1 0.2>
    componentLinearVelScale=<0.1 0.1 0.2>
    gravityOverride {z=-980}
    physicsBodyDefinitions [{
      sphereCollisionRadius=0
      boundBone {boneName=dyn_necklace_03}
      boxExtents=<16 4 16>
      constraintSetup {angularLimitsMax=<0 3 15>, angularLimitsMin=<0 -3 -2>, angularTarget=<1 0 0>}
      localJointOffset=<2 0 0>
     }]
   }
   n12 = Anim Dynamics - Bone: dyn_necklace_04(ComponentPose=@n11, Alpha=0.7, ExternalForce="0,0,0") {
    angularDampingOverride=0.9999
    bOverrideAngularDamping=true
    bUseGravityOverride=true
    bUsePlanarLimit=false
    bUseSphericalLimits=true
    alphaBoolBlend {blendOption=Linear}
    boundBone {boneName=dyn_necklace_04}
    componentAppliedLinearAccClamp=<1000 1000 1000>
    componentLinearAccScale=<0.1 0.1 0.2>
    componentLinearVelScale=<0.1 0.1 0.2>
    gravityOverride {z=-980}
    physicsBodyDefinitions [{
      sphereCollisionRadius=0
      boundBone {boneName=dyn_necklace_04}
      boxExtents=<15 2 15>
      constraintSetup {angularLimitsMax=<0 15 15>, angularLimitsMin=<0 -15 0>, angularTarget=<1 0 0>}
      localJointOffset=<2 0 0>
     }]
    sphericalLimits [{limitRadius=20, drivingBone {boneName=spine_04}, sphereLocalOffset=<0 -7.5 0>}]
   }
   n13 = Anim Dynamics - Bone: dyn_shoelace_r_bowR(ComponentPose=@n12, Alpha=1, ExternalForce="0,0,0") {
    angularDampingOverride=0.999
    bOverrideAngularDamping=true
    bUseGravityOverride=true
    bUsePlanarLimit=false
    simulationSpace=RootRelative
    alphaBoolBlend {blendOption=Linear}
    boundBone {boneName=dyn_shoelace_r_bowR}
    componentAppliedLinearAccClamp=<1500 1500 1500>
    componentLinearAccScale=<0.2 0.2 0.3>
    componentLinearVelScale=<0.2 0.2 0.3>
    gravityOverride {z=-400}
    physicsBodyDefinitions [{
      boundBone {boneName=dyn_shoelace_r_bowR}
      boxExtents=<11 11 6>
      constraintSetup {angularLimitsMax=<0 30 2>, angularLimitsMin=<0 0 -20>, angularTarget=<0 0 0>}
      localJointOffset=<2 0 0>
     }]
    relativeSpaceBone {boneName=calf_r}
   }
   n14 = Anim Dynamics - Bone: dyn_shoelace_r_bowL(ComponentPose=@n13, Alpha=1, ExternalForce="0,0,0") {
    angularDampingOverride=0.999
    bOverrideAngularDamping=true
    bUseGravityOverride=true
    simulationSpace=RootRelative
    alphaBoolBlend {blendOption=Linear}
    boundBone {boneName=dyn_shoelace_r_bowL}
    componentAppliedLinearAccClamp=<1500 1500 1500>
    componentLinearAccScale=<0.2 0.2 0.3>
    componentLinearVelScale=<0.2 0.2 0.3>
    gravityOverride {z=-400}
    physicsBodyDefinitions [{
      boundBone {boneName=dyn_shoelace_r_bowL}
      boxExtents=<10 10 5>
      constraintSetup {angularLimitsMax=<0 0 2>, angularLimitsMin=<0 -30 -20>, angularTarget=<0 0 0>}
      localJointOffset=<2 0 0>
     }]
    relativeSpaceBone {boneName=calf_r}
   }
   n15 = Anim Dynamics - Bone: dyn_shoelace_r_strings(ComponentPose=@n14, Alpha=1, ExternalForce="0,0,0") {
    angularDampingOverride=0.999
    bOverrideAngularDamping=true
    bUseGravityOverride=true
    simulationSpace=RootRelative
    alphaBoolBlend {blendOption=Linear}
    boundBone {boneName=dyn_shoelace_r_strings}
    componentAppliedLinearAccClamp=<1500 1500 1500>
    componentLinearAccScale=<0.2 0.2 0.3>
    componentLinearVelScale=<0.2 0.2 0.3>
    gravityOverride {z=-750}
    physicsBodyDefinitions [{
      boundBone {boneName=dyn_shoelace_r_strings}
      boxExtents=<14 14 10>
      constraintSetup {angularLimitsMax=<0 0 10>, angularLimitsMin=<0 -40 -10>, angularTarget=<0 0 0>}
      localJointOffset=<4 0 0>
     }]
    relativeSpaceBone {boneName=calf_r}
   }
   n16 = Anim Dynamics - Bone: dyn_shoelace_l_bowR(ComponentPose=@n15, Alpha=1, ExternalForce="0,0,0") {
    angularDampingOverride=0.999
    bOverrideAngularDamping=true
    bUseGravityOverride=true
    bUsePlanarLimit=false
    simulationSpace=RootRelative
    alphaBoolBlend {blendOption=Linear}
    boundBone {boneName=dyn_shoelace_l_bowR}
    componentAppliedLinearAccClamp=<1500 1500 1500>
    componentLinearAccScale=<0.2 0.2 0.3>
    componentLinearVelScale=<0.2 0.2 0.3>
    gravityOverride {z=-400}
    physicsBodyDefinitions [{
      boundBone {boneName=dyn_shoelace_l_bowR}
      boxExtents=<11 11 6>
      constraintSetup {angularLimitsMax=<0 0 2>, angularLimitsMin=<0 -30 -20>, angularTarget=<0 0 0>}
      localJointOffset=<-2 0 0>
     }]
    relativeSpaceBone {boneName=calf_r}
   }
   n17 = Anim Dynamics - Bone: dyn_shoelace_l_bowL(ComponentPose=@n16, Alpha=1, ExternalForce="0,0,0") {
    angularDampingOverride=0.999
    bOverrideAngularDamping=true
    bUseGravityOverride=true
    simulationSpace=RootRelative
    alphaBoolBlend {blendOption=Linear}
    boundBone {boneName=dyn_shoelace_l_bowL}
    componentAppliedLinearAccClamp=<1500 1500 1500>
    componentLinearAccScale=<0.2 0.2 0.3>
    componentLinearVelScale=<0.2 0.2 0.3>
    gravityOverride {z=-400}
    physicsBodyDefinitions [{
      boundBone {boneName=dyn_shoelace_l_bowL}
      boxExtents=<10 10 5>
      constraintSetup {angularLimitsMax=<0 30 2>, angularLimitsMin=<0 0 -20>, angularTarget=<0 0 0>}
      localJointOffset=<-2 0 0>
     }]
    relativeSpaceBone {boneName=calf_r}
   }
   n18 = Anim Dynamics - Bone: dyn_shoelace_l_strings(ComponentPose=@n17, Alpha=1, ExternalForce="0,0,0") {
    angularDampingOverride=0.999
    bOverrideAngularDamping=true
    bUseGravityOverride=true
    simulationSpace=RootRelative
    alphaBoolBlend {blendOption=Linear}
    boundBone {boneName=dyn_shoelace_l_strings}
    componentAppliedLinearAccClamp=<1500 1500 1500>
    componentLinearAccScale=<0.2 0.2 0.3>
    componentLinearVelScale=<0.2 0.2 0.3>
    gravityOverride {z=-750}
    physicsBodyDefinitions [{
      boundBone {boneName=dyn_shoelace_l_strings}
      boxExtents=<14 14 10>
      constraintSetup {angularLimitsMax=<0 0 10>, angularLimitsMin=<0 -40 -10>, angularTarget=<0 0 0>}
      localJointOffset=<-4 0 0>
     }]
    relativeSpaceBone {boneName=calf_r}
   }
   Output Pose(Result=Component To Local(ComponentPose=@n18)) {layerGroup=None}
   n19 = Get Crouch Alpha() -> (Crouch Alpha:double)
   n20 = Control Rig(Source=Copy Pose From Mesh() {bCopyCurves=true, bUseAttachedParent=true}) {
    alphaBoolBlend {blendOption=Linear}
    controlRigAssetReference {blueprintRigClass=&/F_MED_Cadet_Deformer_CtrlRig.F_MED_Cadet_Deformer_CtrlRig_C}
   }
   n21 = "Transform (Modify) Bone - Bone: dyn_belt_rFR_MSTR"(ComponentPose=Local To Component(LocalPose=@n20), Alpha=@n19, Translation="0,0,4", Rotation="0,-5,0", Scale="1,1,1") {
    rotationMode=BMM_Additive
    rotationSpace=BCS_BoneSpace
    scaleMode=BMM_Additive
    scaleSpace=BCS_BoneSpace
    translationMode=BMM_Additive
    translationSpace=BCS_BoneSpace
    alphaBoolBlend {blendOption=Linear}
    alphaScaleBiasClamp {bInterpResult=true, interpSpeedDecreasing=8.5, interpSpeedIncreasing=8.5}
    boneToModify {boneName=dyn_belt_rFR_MSTR}
    rotation {yaw=-5}
    translation {z=4}
   }
   n22 = "Transform (Modify) Bone - Bone: dyn_belt_rBK_MSTR"(ComponentPose=@n21, Alpha=@n19, Translation="0,0,2", Rotation="0,0,0", Scale="1,1,1") {
    rotationMode=BMM_Additive
    rotationSpace=BCS_BoneSpace
    scaleMode=BMM_Additive
    scaleSpace=BCS_BoneSpace
    translationMode=BMM_Additive
    translationSpace=BCS_BoneSpace
    alphaBoolBlend {blendOption=Linear}
    alphaScaleBiasClamp {bInterpResult=true, interpSpeedDecreasing=8.5, interpSpeedIncreasing=8.5}
    boneToModify {boneName=dyn_belt_rBK_MSTR}
    translation {z=2}
   }
   n23 = Get R Leg Alpha() -> (R Leg Alpha:double)
   n24 = "Transform (Modify) Bone - Bone: dyn_belt_rFR_MSTR"(ComponentPose=@n22, Alpha=@n23, Translation="0,0,2.75", Rotation="-30,12,-75", Scale="1,0.9,1") {
    rotationMode=BMM_Additive
    rotationSpace=BCS_BoneSpace
    scaleMode=BMM_Additive
    scaleSpace=BCS_BoneSpace
    translationMode=BMM_Additive
    translationSpace=BCS_BoneSpace
    alphaBoolBlend {blendOption=Linear}
    alphaScaleBiasClamp {bInterpResult=true, interpSpeedDecreasing=50, interpSpeedIncreasing=50}
    boneToModify {boneName=dyn_belt_rFR_MSTR}
    rotation=<rot -30 12 -75>
    scale {y=0.9}
    translation {z=2.75}
   }
   n25 = "Transform (Modify) Bone - Bone: dyn_belt_rBK_MSTR"(ComponentPose=@n24, Alpha=@n23, Translation="0,0,2", Rotation="-10,0,0", Scale="1,1,1") {
    rotationMode=BMM_Additive
    rotationSpace=BCS_BoneSpace
    scaleMode=BMM_Additive
    scaleSpace=BCS_BoneSpace
    translationMode=BMM_Additive
    translationSpace=BCS_BoneSpace
    alphaBoolBlend {blendOption=Linear}
    alphaScaleBiasClamp {bInterpResult=true, interpSpeedDecreasing=50, interpSpeedIncreasing=50}
    boneToModify {boneName=dyn_belt_rBK_MSTR}
    rotation {pitch=-10}
    translation {z=2}
   }
   n26 = Get R Leg Crouch Alpha() -> (R Leg Crouch Alpha:double)
   n27 = "Transform (Modify) Bone - Bone: dyn_belt_rFR_MSTR"(ComponentPose=@n25, Alpha=@n26, Translation="0,0,0", Rotation="-30,6,-80", Scale="1,0.9,1") {
    rotationMode=BMM_Additive
    rotationSpace=BCS_BoneSpace
    scaleMode=BMM_Additive
    scaleSpace=BCS_BoneSpace
    translationMode=BMM_Additive
    translationSpace=BCS_BoneSpace
    alphaBoolBlend {blendOption=Linear}
    alphaScaleBiasClamp {bInterpResult=true, interpSpeedDecreasing=50, interpSpeedIncreasing=50}
    boneToModify {boneName=dyn_belt_rFR_MSTR}
    rotation=<rot -30 6 -80>
    scale {y=0.9}
   }
   n28 = "Transform (Modify) Bone - Bone: dyn_belt_rBK_MSTR"(ComponentPose=@n27, Alpha=@n26, Translation="0,0,0", Rotation="-5,0,0", Scale="1,1,1") {
    rotationMode=BMM_Additive
    rotationSpace=BCS_BoneSpace
    scaleMode=BMM_Additive
    scaleSpace=BCS_BoneSpace
    translationMode=BMM_Additive
    translationSpace=BCS_BoneSpace
    alphaBoolBlend {blendOption=Linear}
    alphaScaleBiasClamp {bInterpResult=true, interpSpeedDecreasing=30, interpSpeedIncreasing=30}
    boneToModify {boneName=dyn_belt_rBK_MSTR}
    rotation {pitch=-5}
   }
   n29 = "Transform (Modify) Bone - Bone: dyn_belt_rFR_MSTR"(
    ComponentPose=@n28
    Alpha=Get R Leg Out Crouch Alpha() -> (R Leg Out Crouch Alpha:double)
    Translation="0,0,1"
    Rotation="-25,0,-10"
    Scale="1,1,1"
   ) {
    rotationMode=BMM_Additive
    rotationSpace=BCS_BoneSpace
    scaleMode=BMM_Additive
    scaleSpace=BCS_BoneSpace
    translationMode=BMM_Additive
    translationSpace=BCS_BoneSpace
    alphaBoolBlend {blendOption=Linear}
    alphaScaleBiasClamp {bInterpResult=true, interpSpeedDecreasing=50, interpSpeedIncreasing=50}
    boneToModify {boneName=dyn_belt_rFR_MSTR}
    rotation {pitch=-25, roll=-10}
    translation {z=1}
   }
   n30 = "Constraint - dyn_beltPouch_a_MSTR (PPPPPPPP)"!(
    ComponentPose=@n29
    Alpha=1
    ConstraintWeights_0=0.7899
    ConstraintWeights_1:float=0.0645
    ConstraintWeights_2:float=0.0675
    ConstraintWeights_3:float=0.0042
    ConstraintWeights_4:float=0.0352
    ConstraintWeights_5:float=0.0383
    ConstraintWeights_6:float=0.0001
    ConstraintWeights_7:float=0.0003
   ) -> (Pose) {
    alphaBoolBlend {blendOption=Linear}
    boneToModify {boneName=dyn_beltPouch_a_MSTR}
    constraintSetup [
     {targetBone {boneName=pelvis}}
     {targetBone {boneName=spine_01}}
     {targetBone {boneName=spine_02}}
     {targetBone {boneName=spine_03}}
     {targetBone {boneName=thigh_l}}
     {targetBone {boneName=thigh_twist_01_l}}
     {targetBone {boneName=thigh_r}}
     {targetBone {boneName=thigh_twist_01_r}}
    ]
    constraintWeights [1,1,1,1,1,1,1,1]
   }
   n31 = "Constraint - dyn_beltPouch_b_MSTR (PPPPPPPP)"!(
    ComponentPose=@n30
    Alpha=1
    ConstraintWeights_0=0.8294
    ConstraintWeights_1:float=0.0599
    ConstraintWeights_2:float=0.0519
    ConstraintWeights_3:float=0.0015
    ConstraintWeights_4:float=0.0274
    ConstraintWeights_5:float=0.025
    ConstraintWeights_6:float=0.0021
    ConstraintWeights_7:float=0.0028
   ) -> (Pose) {
    alphaBoolBlend {blendOption=Linear}
    boneToModify {boneName=dyn_beltPouch_b_MSTR}
    constraintSetup [
     {targetBone {boneName=pelvis}}
     {targetBone {boneName=spine_01}}
     {targetBone {boneName=spine_02}}
     {targetBone {boneName=spine_03}}
     {targetBone {boneName=thigh_l}}
     {targetBone {boneName=thigh_twist_01_l}}
     {targetBone {boneName=thigh_r}}
     {targetBone {boneName=thigh_twist_01_r}}
    ]
    constraintWeights [1,1,1,1,1,1,1,1]
   }
   n32 = "Constraint - dyn_necklace_MSTR (PP)"!(ComponentPose=@n31, Alpha=1, ConstraintWeights_0=0.5, ConstraintWeights_1:float=0.5) -> (Pose) {
    alphaBoolBlend {blendOption=Linear}
    boneToModify {boneName=dyn_necklace_MSTR}
    constraintSetup [{targetBone {boneName=spine_04}}, {targetBone {boneName=spine_05}}]
    constraintWeights [1,1]
   }
   n33 = "Constraint - dyn_knife_MSTR (PPP)"!(ComponentPose=@n32, Alpha=1, ConstraintWeights_0=0.1217, ConstraintWeights_1:float=0.1208, ConstraintWeights_2:float=0.7575) -> (Pose) {
    alphaBoolBlend {blendOption=Linear}
    boneToModify {boneName=dyn_knife_MSTR}
    constraintSetup [{targetBone {boneName=calf_l}}, {targetBone {boneName=calf_twist_01_l}}, {targetBone {boneName=calf_twist_02_l}}]
    constraintWeights [1,1,1]
   }
   n34 = "Constraint - dyn_legBullet (PPPP)"!(
    ComponentPose=@n33
    Alpha=1
    ConstraintWeights_0=0.0408
    ConstraintWeights_1:float=0.1881
    ConstraintWeights_2:float=0.6623
    ConstraintWeights_3:float=0.1088
   ) -> (Pose) {
    alphaBoolBlend {blendOption=Linear}
    boneToModify {boneName=dyn_legBullet}
    constraintSetup [
     {targetBone {boneName=thigh_twist_01_r}}
     {targetBone {boneName=calf_r}}
     {targetBone {boneName=calf_twist_02_r}}
     {targetBone {boneName=deform_knee_bk_01_r}}
    ]
    constraintWeights [1,1,1,1]
   }
   n35 = "Constraint - dyn_shoelace_r_MSTR (PPP)"!(ComponentPose=@n34, Alpha=1, ConstraintWeights_0=0.1024, ConstraintWeights_1:float=0.1263, ConstraintWeights_2:float=0.7714) -> (Pose) {
    alphaBoolBlend {blendOption=Linear}
    boneToModify {boneName=dyn_shoelace_r_MSTR}
    constraintSetup [{targetBone {boneName=calf_r}}, {targetBone {boneName=calf_twist_01_r}}, {targetBone {boneName=calf_twist_02_r}}]
    constraintWeights [1,1,1]
   }
   n36 = "Constraint - dyn_shoelace_l_MSTR (PPP)"!(ComponentPose=@n35, Alpha=1, ConstraintWeights_0=0.1011, ConstraintWeights_1:float=0.1273, ConstraintWeights_2:float=0.7717) -> (Pose) {
    alphaBoolBlend {blendOption=Linear}
    boneToModify {boneName=dyn_shoelace_l_MSTR}
    constraintSetup [{targetBone {boneName=calf_l}}, {targetBone {boneName=calf_twist_01_l}}, {targetBone {boneName=calf_twist_02_l}}]
    constraintWeights [1,1,1]
   }
   n37 = "Constraint - dyn_r_handSpike_3 (PPPPPPPP)"!(
    ComponentPose=@n36
    Alpha=1
    ConstraintWeights_0=0.3407
    ConstraintWeights_1:float=0.012
    ConstraintWeights_2:float=0.0038
    ConstraintWeights_3:float=0.0045
    ConstraintWeights_4:float=0.1071
    ConstraintWeights_5:float=0.4483
    ConstraintWeights_6:float=0.076
    ConstraintWeights_7:float=0.0077
   ) -> (Pose) {
    alphaBoolBlend {blendOption=Linear}
    boneToModify {boneName=dyn_r_handSpike_3}
    constraintSetup [
     {targetBone {boneName=ring_01_r}}
     {targetBone {boneName=ring_02_r}}
     {targetBone {boneName=pinky_03_r}}
     {targetBone {boneName=hand_r}}
     {targetBone {boneName=pinky_metacarpal_r}}
     {targetBone {boneName=ring_metacarpal_r}}
     {targetBone {boneName=middle_metacarpal_r}}
     {targetBone {boneName=index_metacarpal_r}}
    ]
    constraintWeights [1,1,1,1,1,1,1,1]
   }
   n38 = "Constraint - dyn_r_handSpike_2 (PPPPPPPP)"!(
    ComponentPose=@n37
    Alpha=1
    ConstraintWeights_0=0.3867
    ConstraintWeights_1:float=0.0096
    ConstraintWeights_2:float=0.0031
    ConstraintWeights_3:float=0.0081
    ConstraintWeights_4:float=0.0475
    ConstraintWeights_5:float=0.4777
    ConstraintWeights_6:float=0.0651
    ConstraintWeights_7:float=0.0021
   ) -> (Pose) {
    alphaBoolBlend {blendOption=Linear}
    boneToModify {boneName=dyn_r_handSpike_2}
    constraintSetup [
     {targetBone {boneName=middle_01_r}}
     {targetBone {boneName=middle_02_r}}
     {targetBone {boneName=hand_r}}
     {targetBone {boneName=pinky_metacarpal_r}}
     {targetBone {boneName=ring_metacarpal_r}}
     {targetBone {boneName=middle_metacarpal_r}}
     {targetBone {boneName=index_metacarpal_r}}
     {targetBone {boneName=thumb_01_r}}
    ]
    constraintWeights [1,1,1,1,1,1,1,1]
   }
   n39 = "Constraint - dyn_r_handSpike_1 (PPPPPPPP)"!(
    ComponentPose=@n38
    Alpha=1
    ConstraintWeights_0=0.2825
    ConstraintWeights_1:float=0.0032
    ConstraintWeights_2:float=0.0006
    ConstraintWeights_3:float=0.0006
    ConstraintWeights_4:float=0.0013
    ConstraintWeights_5:float=0.0112
    ConstraintWeights_6:float=0.6999
    ConstraintWeights_7:float=0.0008
   ) -> (Pose) {
    alphaBoolBlend {blendOption=Linear}
    boneToModify {boneName=dyn_r_handSpike_1}
    constraintSetup [
     {targetBone {boneName=index_01_r}}
     {targetBone {boneName=index_02_r}}
     {targetBone {boneName=index_03_r}}
     {targetBone {boneName=thumb_03_r}}
     {targetBone {boneName=ring_metacarpal_r}}
     {targetBone {boneName=middle_metacarpal_r}}
     {targetBone {boneName=index_metacarpal_r}}
     {targetBone {boneName=thumb_01_r}}
    ]
    constraintWeights [1,1,1,1,1,1,1,1]
   }
   n40 = "Constraint - dyn_l_handSpike_3 (PPPPPPPP)"!(
    ComponentPose=@n39
    Alpha=1
    ConstraintWeights_0=0.3407
    ConstraintWeights_1:float=0.012
    ConstraintWeights_2:float=0.0038
    ConstraintWeights_3:float=0.0045
    ConstraintWeights_4:float=0.4483
    ConstraintWeights_5:float=0.076
    ConstraintWeights_6:float=0.0077
    ConstraintWeights_7:float=0.1071
   ) -> (Pose) {
    alphaBoolBlend {blendOption=Linear}
    boneToModify {boneName=dyn_l_handSpike_3}
    constraintSetup [
     {targetBone {boneName=ring_01_l}}
     {targetBone {boneName=ring_02_l}}
     {targetBone {boneName=pinky_03_l}}
     {targetBone {boneName=hand_l}}
     {targetBone {boneName=ring_metacarpal_l}}
     {targetBone {boneName=middle_metacarpal_l}}
     {targetBone {boneName=index_metacarpal_l}}
     {targetBone {boneName=pinky_metacarpal_l}}
    ]
    constraintWeights [1,1,1,1,1,1,1,1]
   }
   n41 = "Constraint - dyn_l_handSpike_2 (PPPPPPPP)"!(
    ComponentPose=@n40
    Alpha=1
    ConstraintWeights_0=0.3867
    ConstraintWeights_1:float=0.0096
    ConstraintWeights_2:float=0.0031
    ConstraintWeights_3:float=0.0475
    ConstraintWeights_4:float=0.4777
    ConstraintWeights_5:float=0.0021
    ConstraintWeights_6:float=0.0651
    ConstraintWeights_7:float=0.0081
   ) -> (Pose) {
    alphaBoolBlend {blendOption=Linear}
    boneToModify {boneName=dyn_l_handSpike_2}
    constraintSetup [
     {targetBone {boneName=middle_01_l}}
     {targetBone {boneName=middle_02_l}}
     {targetBone {boneName=hand_l}}
     {targetBone {boneName=ring_metacarpal_l}}
     {targetBone {boneName=middle_metacarpal_l}}
     {targetBone {boneName=thumb_01_l}}
     {targetBone {boneName=index_metacarpal_l}}
     {targetBone {boneName=pinky_metacarpal_l}}
    ]
    constraintWeights [1,1,1,1,1,1,1,1]
   }
   n42 = "Constraint - dyn_l_handSpike_1 (PPPPPPPP)"!(
    ComponentPose=@n41
    Alpha=1
    ConstraintWeights_0=0.2825
    ConstraintWeights_1:float=0.0032
    ConstraintWeights_2:float=0.0006
    ConstraintWeights_3:float=0.0006
    ConstraintWeights_4:float=0.0013
    ConstraintWeights_5:float=0.0112
    ConstraintWeights_6:float=0.0008
    ConstraintWeights_7:float=0.6999
   ) -> (Pose) {
    alphaBoolBlend {blendOption=Linear}
    boneToModify {boneName=dyn_l_handSpike_1}
    constraintSetup [
     {targetBone {boneName=index_01_l}}
     {targetBone {boneName=index_02_l}}
     {targetBone {boneName=index_03_l}}
     {targetBone {boneName=thumb_03_l}}
     {targetBone {boneName=ring_metacarpal_l}}
     {targetBone {boneName=middle_metacarpal_l}}
     {targetBone {boneName=thumb_01_l}}
     {targetBone {boneName=index_metacarpal_l}}
    ]
    constraintWeights [1,1,1,1,1,1,1,1]
   }
   n43 = "Constraint - dyn_elbowPad_l (PPPPPPPP)"!(
    ComponentPose=@n42
    Alpha=1
    ConstraintWeights_0=0.2709
    ConstraintWeights_1:float=0.0046
    ConstraintWeights_2:float=0.2551
    ConstraintWeights_3:float=0.0878
    ConstraintWeights_4:float=0.0813
    ConstraintWeights_5:float=0.2387
    ConstraintWeights_6:float=0.001
    ConstraintWeights_7:float=0.0605
   ) -> (Pose) {
    alphaBoolBlend {blendOption=Linear}
    boneToModify {boneName=dyn_elbowPad_l}
    constraintSetup [
     {targetBone {boneName=lowerarm_l}}
     {targetBone {boneName=lowerarm_twist_02_l}}
     {targetBone {boneName=upperarm_l}}
     {targetBone {boneName=upperarm_twist_02_l}}
     {targetBone {boneName=upperarm_twist_01_l}}
     {targetBone {boneName=deform_elbow_bk_l}}
     {targetBone {boneName=deform_deltoid_fr_l}}
     {targetBone {boneName=deform_elbow_fr_l}}
    ]
    constraintWeights [1,1,1,1,1,1,1,1]
   }
   Save cached pose 'PreSim'(Pose=Component To Local(ComponentPose=@n43)) {cacheName=PreSim}
   Get NeckTurn() -> (NeckTurn:double)
  ]}
 EventGraph {nodes [
   Event BlueprintUpdateAnimation()
   > n1 = Sequence!(execute) -> (then_0, then_1, then_2:Exec)
   > n2 = Gravity Override State Function DEPRECATED(CharacterPartAnimInstance=Self-Reference(), GameTimeInSeconds=GetGameTimeInSeconds(), GravityOverrideParams=Get Gravity Override Params())
   > Set GravityOverride(GravityOverride=@n2)
   n3 = Branch(execute=@n1.then_1, Condition=Get bIsCrouching())
   > Set Crouch Alpha(Crouch Alpha:double=1) -> (Output_Get:double)
   n4 = GetOwnerSkeletalMeshComponent()
   n5 = BreakRotator(InRot=BreakTransform(InTransform=GetDeltaTransformFromRefPose(self=@n4, BoneName=thigh_r, BaseName=pelvis)).Rotation)
   > Set R Leg Crouch Alpha(R Leg Crouch Alpha:double=MapRangeClamped(Value=@n5.Yaw, InRangeA=-25, InRangeB=-100)) -> (Output_Get:double)
   > Set R Leg Alpha(R Leg Alpha:double=0) -> (Output_Get:double)
   > Set R Leg Out Crouch Alpha(R Leg Out Crouch Alpha:double=MapRangeClamped(Value=@n5.Pitch, InRangeA=7, InRangeB=40)) -> (Output_Get:double)
   Set Crouch Alpha(execute=@n3.else, Crouch Alpha=0)
   > Set R Leg Crouch Alpha(R Leg Crouch Alpha=0)
   n6 = MapRangeClamped(Value=@n5.Yaw, InRangeA=5, InRangeB=-90)
   n7 = Get bIsOnGround() -> (bIsOnGround:bool)
   > Set R Leg Alpha(R Leg Alpha=SelectFloat(A=@n6, B=float * float(A=@n6, B=0.62) {is=title K2Node_PromotableOperator}, bPickA=NOT Boolean(A=@n7)))
   > Set R Leg Out Crouch Alpha(R Leg Out Crouch Alpha=0)
   n8 = GetOwnerSkeletalMeshComponent()
   n9 = BreakRotator(InRot=BreakTransform(InTransform=GetDeltaTransformFromRefPose(self=@n8, BoneName=neck_01, BaseName=spine_05)).Rotation)
   Set NeckTurn(execute=@n1.then_2, NeckTurn:double=MapRangeClamped(Value="Absolute (Float)"(A=@n9), InRangeA=0, InRangeB=30, OutRangeA=0.5)) -> (Output_Get:double)
   Event OnWeaponChanged()
   > Set isWeaponSwapping(isWeaponSwapping:bool=true) -> (Output_Get:bool)
   > Delay(Duration=0.1)
   > Set isWeaponSwapping(isWeaponSwapping=false)
   TryGetPawnOwner()
  ]}
]
variables [
 Crouch Alpha {type=number}
 Gravity Override Params {
  type=GravityOverrideParamsStruct
  default {
   skydive_FloatingDown {gravityOverride=<0 0 -720>}
   skydive_GoingUp {gravityOverride=<0 0 -200>}
   skydive_Parachuting {gravityOverride=<0 0 -720>, windAmplitude=<100 -250 200>, windFrequency=<0.75 3.5 4>, windOffset=<0 -300 200>}
  }
 }
 GravityOverride {type=Vector}
 isWeaponSwapping {type=boolean}
 NeckTurn {type=number}
 R Leg Alpha {type=number}
 R Leg Crouch Alpha {type=number}
 R Leg Out Crouch Alpha {type=number}
]
