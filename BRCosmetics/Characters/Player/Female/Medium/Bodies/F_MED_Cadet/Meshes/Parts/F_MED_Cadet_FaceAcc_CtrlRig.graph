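# Control Rig for the face accessory part of the medium female Cadet, orienting the two hair braids relative to the upper body.
# On execution, two chained AimItem units point the X axis of the left, then right, braid master bone at separate fixed offsets in spine_05 space.
# The braid roots thus track the chest although parented under the head.
# Secondary axes are zero, and the child braid and earring bones have no nodes of their own.
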ControlRigBlueprint /BRCosmetics/Characters/Player/Female/Medium/Bodies/F_MED_Cadet/Meshes/Parts/F_MED_Cadet_FaceAcc_CtrlRig
graphs [
 RigVMFunctionLibrary {links [], nodes []}
 RigVMModel {nodes [
   RigUnit_BeginExecution = &RigUnit_BeginExecution()
   /Script/ControlRig.RigUnit_AimItem(
    DebugSettings=(Scale=10,WorldOffset=(Rotation=(X=0,Y=0,Z=0,W=1),Translation=(X=0,Y=0,Z=0),Scale3D=(X=1,Y=1,Z=1)))
    Item=(Type=Bone,Name="dyn_l_braid_MSTR")
    Primary=(Weight=1,Axis=(X=1,Y=0,Z=0),Target=(X=0,Y=-8,Z=-5),Kind=Location,Space=(Type=Bone,Name="spine_05"))
    Secondary=(Weight=1,Axis=(X=0,Y=0,Z=0),Target=(X=0,Y=0,Z=0),Kind=Location,Space=(Type=Bone))
    Weight=1
    ExecutePin=@RigUnit_BeginExecution.ExecutePin
   ) {is=struct RigVMUnitNode}
   > &RigUnit_AimItem(
    DebugSettings=(Scale=10,WorldOffset=(Rotation=(X=0,Y=0,Z=0,W=1),Translation=(X=0,Y=0,Z=0),Scale3D=(X=1,Y=1,Z=1)))
    Item=(Type=Bone,Name="dyn_r_braid_MSTR")
    Primary=(Weight=1,Axis=(X=1,Y=0,Z=0),Target=(X=0,Y=-6,Z=3),Kind=Location,Space=(Type=Bone,Name="spine_05"))
    Secondary=(Weight=1,Axis=(X=0,Y=0,Z=0),Target=(X=0,Y=0,Z=0),Kind=Location,Space=(Type=Bone))
    Weight=1
   )
  ]}
]
hierarchy [
 root {loc=<0 0 0>, rot=<0 0 0 1>}
 pelvis {parent=root, loc=<0 -0.5301 98.0571>, rot=<0.0127 0.707 -0.0127 -0.707>}
 spine_01 {parent=pelvis, loc=<6.1208 -0.9328 0>, rot=<0 0 0.034 -0.9994>}
 spine_02 {parent=spine_01, loc=<6.5996 -0.2095 0>, rot=<0 0 0.0238 -0.9997>}
 spine_03 {parent=spine_02, loc=<6.4914 -0.1237 0>, rot=<0 0 0.0149 -0.9999>}
 spine_04 {parent=spine_03, loc=<6.9598 0.0794 0>, rot=<0 0 -0.0132 -0.9999>}
 spine_05 {parent=spine_04, loc=<6.9987 0 0>, rot=<0 0 -0.0015 -1>}
 clavicle_l {parent=spine_05, loc=<5.4831 -0.1053 -3.2946>, rot=<0.0465 -0.7289 -0.009 -0.6829>}
 upperarm_l {parent=clavicle_l, loc=<9.8658 -1.2136 -0.4105>, rot=<0.0164 -0.3245 -0.0269 -0.9454>}
 clavicle_r {parent=spine_05, loc=<5.4831 -0.1053 3.2946>, rot=<0.7289 0.0465 -0.6829 0.009>}
 upperarm_r {parent=clavicle_r, loc=<-9.8658 1.2136 0.4104>, rot=<0.0164 -0.3245 -0.0269 -0.9454>}
 neck_01 {parent=spine_05, loc=<11.3138 0.1529 0>, rot=<0 0 -0.2003 -0.9797>}
 neck_02 {parent=neck_01, loc=<4.5627 0 0>, rot=<0 0 0.0504 -0.9987>}
 head {parent=neck_02, loc=<5.3914 0 0>, rot=<0 0 0.115 -0.9934>}
 dyn_Hair_Braid_MSTR {parent=head, loc=<0 0 0>, rot=<0 0 0 1>}
 dyn_r_braid_MSTR {parent=dyn_Hair_Braid_MSTR, loc=<9.4381 -5.2419 4.3432>, rot=<0.0065 -1 0 0>}
 dyn_r_braid_01 {parent=dyn_r_braid_MSTR, loc=<0 0 0>, rot=<0 0.0054 -0.0118 -0.9999>}
 dyn_r_braid_02 {parent=dyn_r_braid_01, loc=<10.3753 0 0>, rot=<0.0055 -0.0278 0.2118 -0.9769>}
 dyn_r_braid_03 {parent=dyn_r_braid_02, loc=<10.1307 0 0>, rot=<0.0021 -0.0044 0.0317 -0.9995>}
 dyn_r_braid_04 {parent=dyn_r_braid_03, loc=<8.6659 0 0>, rot=<-0.0062 0.0175 -0.1298 -0.9914>}
 dyn_r_braid_05 {parent=dyn_r_braid_04, loc=<10.1868 0 0>, rot=<-0.0008 0.0043 -0.0338 -0.9994>}
 dyn_r_braid_06 {parent=dyn_r_braid_05, loc=<10.0394 0 0>, rot=<-0.063 -0.0016 0.0248 -0.9977>}
 dyn_l_braid_MSTR {parent=dyn_Hair_Braid_MSTR, loc=<9.438 -5.242 -4.343>, rot=<0.0041 -1 0 0>}
 dyn_l_braid_01 {parent=dyn_l_braid_MSTR, loc=<0 0 0>, rot=<-0.0001 -0.0058 -0.0117 -0.9999>}
 dyn_l_braid_02 {parent=dyn_l_braid_01, loc=<10.3753 0 0>, rot=<-0.0054 0.0283 0.2118 -0.9769>}
 dyn_l_braid_03 {parent=dyn_l_braid_02, loc=<10.1307 0 0>, rot=<-0.0021 0.0045 0.0317 -0.9995>}
 dyn_l_braid_04 {parent=dyn_l_braid_03, loc=<8.6659 0 0>, rot=<0.0062 -0.0178 -0.1298 -0.9914>}
 dyn_l_braid_05 {parent=dyn_l_braid_04, loc=<10.1868 0 0>, rot=<0.0005 -0.0029 -0.0221 -0.9998>}
 dyn_l_braid_06 {parent=dyn_l_braid_05, loc=<9.9774 0 0>, rot=<0.0644 0.0008 0.013 -0.9978>}
 dyn_earring_R_MSTR {parent=head, loc=<0.0333 0.2881 6.2833>, rot=<0.0041 -1 0 0>}
 dyn_earring_R_01 {parent=dyn_earring_R_MSTR, loc=<0 0 0>, rot=<0 0 0 1>}
 dyn_earring_L_MSTR {parent=head, loc=<0.167 0.4199 -6.3037>, rot=<0.0041 -1 0 0>}
 dyn_earring_L_01 {parent=dyn_earring_L_MSTR, loc=<0 0 0>, rot=<0 0 0 1>}
]
