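# Physics asset for the Cadet female medium FaceAcc mesh, covering two hair braids and two belt pouches.
# Capsule bodies cover these dynamic bones plus head, neck, spine, pelvis and upper arms.
# Each six-segment braid hangs from master bones on the head, with very tight twist and swing limits that widen toward the tip.
# Free-angular constraints cross-link the lower left and right braid segments and attach the pouches to the pelvis.
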PhysicsAsset /BRCosmetics/Characters/Player/Female/Medium/Bodies/F_MED_Cadet/Meshes/Parts/F_MED_Cadet_FaceAcc_Physics
bodies {
 dyn_Hair_Braid_MSTR [{shapeType=Capsule, length=0.505, radius=0.505, shapeName=dyn_Hair_Braid_MSTR_capsule, rotation=<rot 0 -90 0>}]
 dyn_beltPouch_a [{shapeType=Capsule, length=8.375, radius=4.625, shapeName=dyn_beltPouch_a_capsule, center=<4.4177 0.5 0>, rotation=<rot 90 0 0>}]
 dyn_beltPouch_b [{shapeType=Capsule, length=8.375, radius=4.625, shapeName=dyn_beltPouch_b_capsule, center=<4.4177 0.5 0>, rotation=<rot 90 0 0>}]
 dyn_l_braid_01 [{shapeType=Capsule, length=8.125, radius=2.5, shapeName=dyn_l_braid_02_capsule, center=<6 -1.5 0>, rotation=<rot 90 0 0>}]
 dyn_l_braid_02 [{shapeType=Capsule, length=9.375, radius=3.25, shapeName=dyn_l_braid_02_capsule, center=<6.75 -1 0>, rotation=<rot 90 0 0>}]
 dyn_l_braid_03 [{shapeType=Capsule, length=8.25, radius=3, shapeName=dyn_l_braid_02_capsule, center=<6.5 -0.5 0>, rotation=<rot 90 0 0>}]
 dyn_l_braid_04 [{shapeType=Capsule, length=7.75, radius=2.5, shapeName=dyn_l_braid_02_capsule, center=<6.5 0 0>, rotation=<rot 90 0 0>}]
 dyn_l_braid_05 [{shapeType=Capsule, length=7.875, radius=2.125, shapeName=dyn_l_braid_02_capsule, center=<6.5 0 0>, rotation=<rot 90 0 0>}]
 dyn_l_braid_06 [{shapeType=Capsule, length=3.125, radius=1.75, shapeName=dyn_l_braid_02_capsule, center=<2.6701 0.0001 -0.0001>, rotation=<rot 90 0 0>}]
 dyn_l_braid_MSTR [{shapeType=Capsule, length=0.505, radius=0.505, shapeName=dyn_l_braid_MSTR_capsule, rotation=<rot 0 -90 0>}]
 dyn_r_braid_01 [{shapeType=Capsule, length=8.125, radius=2.5, shapeName=dyn_l_braid_02_capsule, center=<6 -1.5 0>, rotation=<rot 90 0 0>}]
 dyn_r_braid_02 [{shapeType=Capsule, length=9.375, radius=3.25, shapeName=dyn_l_braid_02_capsule, center=<6.75 -1 0>, rotation=<rot 90 0 0>}]
 dyn_r_braid_03 [{shapeType=Capsule, length=8.25, radius=3, shapeName=dyn_l_braid_02_capsule, center=<6.5 -0.5 0>, rotation=<rot 90 0 0>}]
 dyn_r_braid_04 [{shapeType=Capsule, length=7.75, radius=2.5, shapeName=dyn_l_braid_02_capsule, center=<6.5 0 0>, rotation=<rot 90 0 0>}]
 dyn_r_braid_05 [{shapeType=Capsule, length=7.875, radius=2.125, shapeName=dyn_l_braid_02_capsule, center=<6.5 0 0>, rotation=<rot 90 0 0>}]
 dyn_r_braid_06 [{shapeType=Capsule, length=3.125, radius=1.75, shapeName=dyn_l_braid_02_capsule, center=<2.6701 0.0001 -0.0001>, rotation=<rot 90 0 0>}]
 dyn_r_braid_MSTR [{shapeType=Capsule, length=0.505, radius=0.505, shapeName=dyn_l_braid_MSTR_capsule, rotation=<rot 0 -90 0>}]
 head [{shapeType=Capsule, length=9.2143, radius=8.9361, center=<1.4235 3.8655 -0.3977>, rotation=<rot 90 0 0>}]
 neck_01 [{shapeType=Capsule, length=4, radius=7, shapeName=neck_01_capsule, center=<5.0089 3.1697 -0.0001>, rotation=<rot 90 0 0>}]
 pelvis [{shapeType=Capsule, length=11.75, radius=15.125, shapeName=pelvis_capsule}]
 spine_01 [{shapeType=Capsule, length=4.875, radius=10.125, shapeName=spine_01_capsule}]
 spine_03 [{shapeType=Capsule, length=4.625, radius=10.375, shapeName=spine_03_capsule, center=<0 1.55 0>}]
 spine_05 [{shapeType=Capsule, length=17.75, radius=10.625, shapeName=spine_05_capsule, center=<0 2.1852 0>}]
 upperarm_l [{shapeType=Capsule, length=22, radius=5.5, shapeName=upperarm_l_capsule, center=<14 0 0>, rotation=<rot 90 0 0>}]
 upperarm_r [{shapeType=Capsule, length=22, radius=5.5, shapeName=upperarm_r_capsule, center=<-14 0 0>, rotation=<rot 90 0 0>}]
}
constraints [
 {
  bone1Name=dyn_Hair_Braid_MSTR
  bone2Name=head
  swing1LimitDegrees=45
  swing1Motion=Free
  swing2LimitDegrees=45
  swing2Motion=Free
  twistLimitDegrees=45
  twistMotion=Free
 }
 {
  bone1Name=dyn_r_braid_MSTR
  bone2Name=dyn_Hair_Braid_MSTR
  swing1LimitDegrees=45
  swing1Motion=Free
  swing2LimitDegrees=45
  swing2Motion=Free
  twistLimitDegrees=45
  twistMotion=Free
 }
 {bone1Name=dyn_r_braid_01, bone2Name=dyn_r_braid_MSTR, swing1LimitDegrees=15, swing2LimitDegrees=8, twistLimitDegrees=2}
 {bone1Name=dyn_r_braid_02, bone2Name=dyn_r_braid_01, swing1LimitDegrees=25, swing2LimitDegrees=20, twistLimitDegrees=2}
 {bone1Name=dyn_r_braid_03, bone2Name=dyn_r_braid_02, swing1LimitDegrees=30, swing2LimitDegrees=35, twistLimitDegrees=2}
 {bone1Name=dyn_r_braid_04, bone2Name=dyn_r_braid_03, swing1LimitDegrees=45, swing2LimitDegrees=45, twistLimitDegrees=2}
 {bone1Name=dyn_r_braid_05, bone2Name=dyn_r_braid_04, swing1LimitDegrees=45, swing2LimitDegrees=45, twistLimitDegrees=2}
 {
  bone1Name=dyn_l_braid_MSTR
  bone2Name=dyn_Hair_Braid_MSTR
  swing1LimitDegrees=45
  swing1Motion=Free
  swing2LimitDegrees=45
  swing2Motion=Free
  twistLimitDegrees=45
  twistMotion=Free
 }
 {bone1Name=dyn_l_braid_01, bone2Name=dyn_l_braid_MSTR, swing1LimitDegrees=15, swing2LimitDegrees=8, twistLimitDegrees=2}
 {bone1Name=dyn_l_braid_02, bone2Name=dyn_l_braid_01, swing1LimitDegrees=25, swing2LimitDegrees=20, twistLimitDegrees=2}
 {bone1Name=dyn_l_braid_03, bone2Name=dyn_l_braid_02, swing1LimitDegrees=30, swing2LimitDegrees=35, twistLimitDegrees=2}
 {bone1Name=dyn_l_braid_04, bone2Name=dyn_l_braid_03, swing1LimitDegrees=45, swing2LimitDegrees=45, twistLimitDegrees=2}
 {bone1Name=dyn_l_braid_05, bone2Name=dyn_l_braid_04, swing1LimitDegrees=45, swing2LimitDegrees=45, twistLimitDegrees=2}
 {bone1Name=dyn_l_braid_06, bone2Name=dyn_l_braid_05, swing1LimitDegrees=45, swing2LimitDegrees=45, twistLimitDegrees=2}
 {
  bone1Name=head
  bone2Name=neck_01
  swing1LimitDegrees=45
  swing1Motion=Free
  swing2LimitDegrees=45
  swing2Motion=Free
  twistLimitDegrees=45
  twistMotion=Free
 }
 {
  bone1Name=spine_05
  bone2Name=spine_03
  swing1LimitDegrees=45
  swing1Motion=Free
  swing2LimitDegrees=45
  swing2Motion=Free
  twistLimitDegrees=45
  twistMotion=Free
 }
 {
  bone1Name=spine_03
  bone2Name=spine_01
  swing1LimitDegrees=45
  swing1Motion=Free
  swing2LimitDegrees=45
  swing2Motion=Free
  twistLimitDegrees=45
  twistMotion=Free
 }
 {
  bone1Name=spine_01
  bone2Name=pelvis
  swing1LimitDegrees=45
  swing1Motion=Free
  swing2LimitDegrees=45
  swing2Motion=Free
  twistLimitDegrees=45
  twistMotion=Free
 }
 {bone1Name=dyn_r_braid_06, bone2Name=dyn_r_braid_05, swing1LimitDegrees=45, swing2LimitDegrees=45, twistLimitDegrees=2}
 {
  bone1Name=neck_01
  bone2Name=spine_05
  swing1LimitDegrees=45
  swing1Motion=Free
  swing2LimitDegrees=45
  swing2Motion=Free
  twistLimitDegrees=45
  twistMotion=Free
 }
 {
  bone1Name=dyn_l_braid_03
  bone2Name=dyn_r_braid_03
  swing1LimitDegrees=45
  swing1Motion=Free
  swing2LimitDegrees=45
  swing2Motion=Free
  twistLimitDegrees=45
  twistMotion=Free
 }
 {
  bone1Name=dyn_l_braid_04
  bone2Name=dyn_r_braid_04
  swing1LimitDegrees=45
  swing1Motion=Free
  swing2LimitDegrees=45
  swing2Motion=Free
  twistLimitDegrees=45
  twistMotion=Free
 }
 {
  bone1Name=dyn_l_braid_05
  bone2Name=dyn_r_braid_05
  swing1LimitDegrees=45
  swing1Motion=Free
  swing2LimitDegrees=45
  swing2Motion=Free
  twistLimitDegrees=45
  twistMotion=Free
 }
 {
  bone1Name=dyn_l_braid_06
  bone2Name=dyn_r_braid_06
  swing1LimitDegrees=45
  swing1Motion=Free
  swing2LimitDegrees=45
  swing2Motion=Free
  twistLimitDegrees=45
  twistMotion=Free
 }
 {
  bone1Name=upperarm_l
  bone2Name=spine_05
  swing1LimitDegrees=45
  swing1Motion=Free
  swing2LimitDegrees=45
  swing2Motion=Free
  twistLimitDegrees=45
  twistMotion=Free
 }
 {
  bone1Name=upperarm_r
  bone2Name=spine_05
  swing1LimitDegrees=45
  swing1Motion=Free
  swing2LimitDegrees=45
  swing2Motion=Free
  twistLimitDegrees=45
  twistMotion=Free
 }
 {
  bone1Name=dyn_beltPouch_a
  bone2Name=pelvis
  swing1LimitDegrees=45
  swing1Motion=Free
  swing2LimitDegrees=45
  swing2Motion=Free
  twistLimitDegrees=45
  twistMotion=Free
 }
 {
  bone1Name=dyn_beltPouch_b
  bone2Name=pelvis
  swing1LimitDegrees=45
  swing1Motion=Free
  swing2LimitDegrees=45
  swing2Motion=Free
  twistLimitDegrees=45
  twistMotion=Free
 }
]
